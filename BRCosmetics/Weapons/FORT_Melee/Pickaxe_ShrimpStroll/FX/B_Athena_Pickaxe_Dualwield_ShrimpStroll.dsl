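; Weapon Blueprint for the ShrimpStroll dual-wield pickaxe cosmetic, extending its parent's swing and impact effects.
; Right and left swings pass the instigator's montage section index to the main and offhand swing VFX as "SwingSectionIndex".
; On impact, it sets that index on the Impact FX variable and flags "HitVegetation?" on the spawned effect, detected by casting the hit actor to bush and tree classes.
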(event EventSwingRight_Common
  ;[0 0 0] Swing
  (|CallParentFunction)
  (bind _section_index (Utilities|SetupSwingMontageSectionIndex (Variables|Actor|GetInstigator))
    (:Out
      (Effects|Components|ParticleSystem|SetIntParameter (CallFunction|GetSwingVFXComponent) "SwingSectionIndex" _section_index))))

(event EventSwingLeft_Common
  ;[0 0 0] Swing
  (|CallParentFunction)
  (bind _section_index (Utilities|SetupSwingMontageSectionIndex (Variables|Actor|GetInstigator))
    (:Out
      (Effects|Components|ParticleSystem|SetIntParameter (CallFunction|GetSwingOffhandVFXComponent) "SwingSectionIndex" _section_index))))

(event EventSwingLeft
  (|CallParentFunction))

(event EventSwingRight
  (|CallParentFunction))

(event FortWeapon|EventOnPlayImpactFX
  (params (HitResult :type Hit%20Result%20Structure%20%28by%20ref%29) (ImpactPhysicalSurface :type EPhysicalSurface%20Enum :default "SurfaceType_Default") (SpawnedPSC :type FXSystem%20Component%20Object%20Reference))
  (bind (_bblockinghit _binitialoverlap _time _distance _location _impactpoint _normal _impactnormal _physmat _hitactor) (Collision|BreakHitResult HitResult))
  (bind (_asapollo_big_bush _bsuccess) (Utilities|Casting|CastToApollo_BigBush _hitactor))
  (bind (_asparent_tree _bsuccess_2) (Utilities|Casting|CastToParent_Tree _hitactor))
  (bind (_asparent_tree_apollo _bsuccess_3) (Utilities|Casting|CastToParent_Tree_Apollo _hitactor))
  (|CallParentFunction HitResult ImpactPhysicalSurface SpawnedPSC)
  (bind _section_index (Utilities|SetupSwingMontageSectionIndex (Variables|Actor|GetInstigator))
    (:Out
      (Effects|Components|ParticleSystem|SetIntParameter (Variables|Default|GetImpact%20FX) "SwingSectionIndex" _section_index)
      (Effects|Components|ParticleSystem|SetBoolParameter SpawnedPSC "HitVegetation?" (Math|Boolean|ORBoolean _bsuccess _bsuccess_2 _bsuccess_3)))))

(fn ConstructionScript
  (|CallParentFunction))
