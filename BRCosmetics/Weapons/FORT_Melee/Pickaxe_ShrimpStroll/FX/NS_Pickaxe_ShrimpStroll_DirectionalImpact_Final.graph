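# Niagara impact effect for the ShrimpStroll pickaxe, throwing a one-shot directional burst of flash, flare with light, spike, smoke, sparks and debris.
# A system DirectionVector steers smoke, sparks, spike and leaves, with SwingSectionIndex also adjusting leaf velocity.
# PhysicalSurface drives debris response through FN_PhysicalSurface, and HitVegetation gates wind- and drag-driven leaves.
# Front-end modifiers shrink sprites and alter spark and spike velocities outside gameplay.
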NiagaraSystem /BRCosmetics/Weapons/FORT_Melee/Pickaxe_ShrimpStroll/FX/NS_Pickaxe_ShrimpStroll_DirectionalImpact_Final
dependencies {
 usedDataInterfaces [
  &NiagaraDataInterfaceCurve
  &NiagaraDataInterfaceVector2DCurve
  &NiagaraDataInterfaceArrayColor
  &NiagaraDataInterfaceArrayFloat
  &NiagaraDataInterfaceArrayFloat3
  &NiagaraDataInterfaceVectorCurve
  &NiagaraDataInterfaceColorCurve
 ]
 usedDynamicInputs [
  &SimulationPosition
  &SetBoolByIntComparison
  &ModuloInt
  /Game/Effects/Niagara/Modules/Utility/FN_DirectionVectorByPickaxeType
 ]
 usedModules [
  &self:SystemSpawnScript.SystemScriptSource.SystemScriptGraph.NiagaraNodeAssignment_0.SetVariables_#1
  &self:SystemSpawnScript.SystemScriptSource.SystemScriptGraph.NiagaraNodeAssignment_1.SetVariables_#2
  &self:SystemSpawnScript.SystemScriptSource.SystemScriptGraph.NiagaraNodeAssignment_2.SetVariables_#3
  &SystemState
 ]
}
emitters {
 Flash_Flare001 {
  emitterUpdateScript [
   &EmitterState()
   &SpawnBurst_Instantaneous(Spawn Count=1, Use Spawn Probability=false)
  ]
  particleSpawnScript [
   &InitializeParticle(
    Color=<rgba 1 0.3938 0 1>
    Lifetime=0.1
    Mass Mode=::NewEnumerator1
    Mesh Scale=&FN_FrontEndVector3Modifier(Gameplay=<2 2 0.8>, Front End=&Multiply_VectorByFloat(Vector=<2 2 0.8>, Float=0.5))
    Mesh Scale Mode=::NewEnumerator4
    Sprite Size=&FN_FrontEndVector2Modifier(Front End:Vector2f=&Multiply_Vector2D_ByFloat(Float=0.5, Vector2D=<400 240>), Gameplay:Vector2f=<400 240>)
    Sprite Size Mode=::NewEnumerator4
    UsePositionOffset=false
    Write Material Random=true
   ) {module=InitializeParticle001}
   &CameraOffset(Camera Offset Amount=50)
   &DynamicMaterialParameters(
    Index 0 Param 1=0.2
    Index 0 Param 2=1
    Module.Param0WriteEnabled=true
    Module.Param1WriteEnabled=true
    Module.Param2WriteEnabled=true
    Module.Param3WriteEnabled=true
   )
  ]
  particleUpdateScript [
   &ParticleState()
   &ScaleMeshSize(Scale Factor=&VectorFromFloat(Value=&FloatFromCurve(FloatCurve=DataInterface() {propertyValues {Curve {keys [<at 0 0.5>,<at 1 1>]}}})))
   &ScaleColor(
    Scale Alpha=&Lerp_Float(A=1, B=0, Alpha=$Particles.NormalizedAge)
    Scale RGB=&VectorFromFloat(Value=2)
    ScaleA=true
    ScaleRGB=true
   )
   &FN_Light_v2(Light Exponent=3, Light Radius=150, LightEnabled=true, Scale Color=1)
   /SetVariables(Particles.DynamicMesh:Vector4f=&MakeVector4(
     X=&FloatFromCurve(FloatCurve=DataInterface() {propertyValues {Curve {keys [{arriveTangent=1, time=0.56, value=0}, <at 1 1>]}}})
     Y=45
     Z=1
     W=1
    )) {isSetParameters=true}
   /SetVariables(Particles.Mesh Color:LinearColor=&ScaleLinearColorByCurve(
     InitialValue=<rgba 1 0.5 0 1>
     ScaleCurve=DataInterface() {propertyValues {AlphaCurve {keys [<at 0 1 RCIM_Cubic>,<at 1 0>]}, BlueCurve {keys [<at 0 1>]}, GreenCurve {keys [<at 0 1>]}, RedCurve {keys [<at 0 1>]}}}
     CurveIndex=$Particles.NormalizedAge
    )) {isSetParameters=true}
  ]
  rendererClasses [
   &NiagaraSpriteRendererProperties
   &NiagaraLightRendererProperties
   &NiagaraMeshRendererProperties
  ]
  renderers [
   {rendererClass=&NiagaraSpriteRendererProperties}
   {rendererClass=&NiagaraLightRendererProperties, rendererIndex=1}
   {rendererClass=&NiagaraMeshRendererProperties, rendererIndex=2}
  ]
 }
 Leaves {
  emitterUpdateScript [
   &EmitterState(Life Cycle Mode=::NewEnumerator1, Loop Behavior=::NewEnumerator1, Loop Delay=0.05, Loop Duration=1, UseLoopDelay=true)
   &SpawnBurst_Instantaneous(
    Spawn Count=&UniformRangedInt(Minimum=3, Maximum=7)
    Spawn Probability=&MakeCustomFloatFromBool(Bool=Linked() {linkedVariable {name=User.HitVegetation?, type=NiagaraBool, defaultValue=~, description=~}}, True Float=1, False Float=0)
    Use Spawn Probability=true
   ) {module=SpawnBurst_Instantaneous001}
  ]
  particleSpawnScript [
   &InitializeParticle(
    Color=&SelectLinearColorFromArray(
     Array Sampling Mode=::NewEnumerator0
     Direct Array Index=?
     Color Selection Array=DataInterface() {propertyValues {ColorData [<rgba 0.342 0.5104 0.1432 1>, <rgba 0.4285 0.6146 0.1444 1>]}}
     Random Seed=0
    )
    Lifetime Max=1.2
    Lifetime Min=0.75
    Lifetime Mode=::NewEnumerator1
    Mass=&Multiply_Float(
     A=&RandomRangeFloat(Minimum=2, Maximum=4)
     B=&FloatFromCurve(
      FloatCurve=DataInterface() {propertyValues {Curve {keys [<at 0 0.5 RCIM_Cubic>, {arriveTangent=1.5048, interpMode=RCIM_Cubic, leaveTangent=1.5048, tangentMode=RCTM_User, time=1, value=1}]}}}
      CurveIndex=&ReturnNormalizedExecIndex()
      Scale Curve=1.5
     )
    )
    Mass Mode=::NewEnumerator1
    Sprite Rotation Mode=::NewEnumerator2
    Sprite Size Max=<8 6>
    Sprite Size Min=<4 2>
    Sprite Size Mode=::NewEnumerator2
    UsePositionOffset=false
    Write Material Random=true
   )
   &ShapeLocation(MaskSpawnGroup=false, Override Seed=false, Sphere Radius=20, UseOwnerScale=true)
   &AddVelocity(Velocity=&Multiply_Vector(
     A=$Output.SphereLocation.SphereNormal
     B=&VectorFromFloat(Value=&UniformRangedFloat(Minimum=12, Maximum=70))
    )) {module=AddVelocity001}
   &AddVelocity(Velocity=&Multiply_Vector(
     A=&UniformRangedVector(
      Minimum=&Multiply_Vector(A=<0 1000 -200>, B=&VectorFromFloat(Value=0.2))
      Maximum=<0 1000 -200>
     )
     B=$System.DirectionVector
    ))
   &AddVelocity(Velocity=&Multiply_Vector(
     A=&MakeVector(
      X=0
      Y=0
      Z=&SelectFloatFromArray(Direct Array Index=$User.SwingSectionIndex, Float Selection Array=DataInterface() {propertyValues {FloatData [0,0,-200,0]}})
     )
     B=$System.DirectionVector
    )) {module=AddVelocity003}
   &CameraOffset(Camera Offset Amount=25)
  ]
  particleUpdateScript [
   &ParticleState()
   &ScaleSpriteSize(
    Scale Factor=&Multiply_Vector2(
     A=&MakeVector2D(
      X=&Multiply_Float(
       A=1
       B=&Sine(
        Normalized Angle=$Emitter.Age
        Period=&Multiply_Float(A=3, B=&Add_Float(A=$Particles.Initial.MaterialRandom, B=0.6))
        Scale=1
        Bias=0
       )
      )
      Y=1
     )
     B=&Vector2DFromFloat(Value=&FloatFromCurve(
       FloatCurve=DataInterface() {propertyValues {Curve {keys [
           {arriveTangent=16.7921, interpMode=RCIM_Cubic, leaveTangent=16.7921, tangentMode=RCTM_User, time=0, value=0.5}
           {arriveTangent=-1.0851, interpMode=RCIM_Cubic, leaveTangent=-1.0851, tangentMode=RCTM_User, time=0.1, value=1}
           {arriveTangent=-1.2169, interpMode=RCIM_Cubic, leaveTangent=-1.2169, tangentMode=RCTM_User, time=0.5, value=0.55}
           <at 1 0.25>
          ]}}}
       Scale Curve=4
      ))
    )
    Scale Sprite Size Mode=::NewEnumerator2
   )
   &SpriteRotationRate(Rotation Rate=&FloatFromCurve(
     FloatCurve=DataInterface() {propertyValues {Curve {keys [
         {interpMode=RCIM_Cubic, tangentMode=RCTM_User, time=0, value=1}
         {interpMode=RCIM_Cubic, leaveTangent=-4.1325, tangentMode=RCTM_Break, time=0.5, value=1}
         {interpMode=RCIM_Cubic, tangentMode=RCTM_User, time=1, value=0}
        ]}}}
     Scale Curve=&RandomRangeFloat(Minimum=180, Maximum=360)
    ))
   &ScaleColor(
    Scale Alpha=&FloatFromCurve(FloatCurve=DataInterface() {propertyValues {Curve {keys [<at -0.0046 1.0113 RCIM_Cubic>,<at 1.0011 0.002 RCIM_Cubic>]}}})
    ScaleA=true
    ScaleRGB=false
   )
   &AccelerationForce(Acceleration=<0 0 -1100>) {enabled=false}
   &PointForce(
    Force Origin=&AddVectorToPosition(Position=/Niagara/DynamicInputs/Helpers/SimulationPosition(), Vector=<0 0 50>)
    Force Strength=1000
    Module.Use Falloff=true
    Module.UseLinearForce=false
   )
   &GravityForce(Gravity=&Multiply_VectorByFloat(
     Vector=<0 0 -920>
     Float=&Lerp_Float(
      A=&RandomRangeFloat(Minimum=0.8, Maximum=1)
      B=&FloatFromCurve(
       FloatCurve=DataInterface() {propertyValues {Curve {keys [<at 0 0.1 RCIM_Cubic>, {arriveTangent=1.5048, interpMode=RCIM_Cubic, leaveTangent=1.5048, tangentMode=RCTM_User, time=1, value=1}]}}}
       CurveIndex=&ReturnNormalizedExecIndex()
       Scale Curve=5
      )
      Alpha=$User.LandBurst
     )
    ))
   &WindForce(Wind Speed=$Emitter.WindVector, Wind Speed Scale=$Emitter.WindIntensity)
   &AerodynamicDrag(
    Drag Shape Curve=DataInterface() {propertyValues {Curve {keys [
        {interpMode=RCIM_Cubic, tangentMode=RCTM_User, time=0, value=0.25}
        {arriveTangent=2.0799, interpMode=RCIM_Cubic, leaveTangent=2.0799, tangentMode=RCTM_User, time=1, value=1}
       ]}}}
    Lift Shape Curve=DataInterface() {propertyValues {Curve {keys [
        {interpMode=RCIM_Cubic, tangentMode=RCTM_User, time=0, value=0}
        {interpMode=RCIM_Cubic, tangentMode=RCTM_User, time=0.2, value=1}
        {arriveTangent=-0.375, interpMode=RCIM_Cubic, leaveTangent=-0.375, time=0.5, value=0.35}
        {tangentMode=RCTM_User, time=1, value=0.7}
       ]}}}
    Pivot Offset Minimum Length:Vector3f=<0.15 0 1>
    Rotation Strength=2.5
    Surface Normal Axis:Vector3f=&RandomRangeVector(Minimum=<-0.5 -0.5 1>, Maximum=<0.5 0.5 1>)
    Use Lift Smoothing=true
    Use Linear Drag=true
    Use Rotational Drag=true
   )
   &SolveForcesAndVelocity(Clamp Velocity=false, Limit Acceleration=false)
   &ScaleSpriteSizeBySpeed(
    Max Scale Factor=<0.4 1>
    Min Scale Factor=<0.2 0.5>
    Sample Scale Factor By Curve=true
    Scale Factor Curve=DataInterface() {propertyValues {Curve {keys [
        {arriveTangent=25.9834, interpMode=RCIM_Cubic, leaveTangent=25.9834, tangentMode=RCTM_User, time=0, value=0}
        {arriveTangent=-2.2509, interpMode=RCIM_Cubic, leaveTangent=-2.2509, tangentMode=RCTM_User, time=0.1, value=2}
        {arriveTangent=-0.1167, interpMode=RCIM_Cubic, leaveTangent=-0.1167, tangentMode=RCTM_User, time=0.6, value=0.25}
       ]}}}
   )
   &AlignSpriteToMeshOrientation(SetAlignment=true, SetFacing=true)
   &DynamicMaterialParameters(
    Index 0 Param 1=&RandomRangeFloat(Minimum=0, Maximum=0.5)
    Module.Index0Param0Write=true
    Module.Index0Param1Write=false
    Module.Index0Param2Write=false
    Module.Index0Param3Write=false
   )
  ]
  rendererClasses [&NiagaraSpriteRendererProperties]
  renderers [{rendererClass=&NiagaraSpriteRendererProperties}]
 }
 MIN_Flash001 {
  emitterUpdateScript [
   &EmitterState()
   &SpawnBurst_Instantaneous(
    Spawn Count=&RandomRangeInt(Minimum=3, Maximum=5)
    Use Spawn Probability=false
    UseLoopCountLimit=false
   )
  ]
  particleSpawnScript [
   &InitializeParticle(
    Lifetime=&RandomRangeFloat(Minimum=0.1, Maximum=0.13)
    Mass Mode=::NewEnumerator1
    Sprite Size=&FN_FrontEndVector2Modifier(
     Front End:Vector2f=&Multiply_Vector2D_ByFloat(Float=0.5, Vector2D=&RandomRangeVector2D(Minimum=<10 25>, Maximum=<15 50>))
     Gameplay:Vector2f=&RandomRangeVector2D(Minimum=<10 25>, Maximum=<15 50>)
    )
    Sprite Size Mode=::NewEnumerator4
    UsePositionOffset=false
    Write Material Random=true
   ) {module=InitializeParticle001}
   &SphereLocation(
    MaskSpawnGroup=false
    Override Seed=false
    Sphere Radius=1
    Surface Only=false
    UseNonUniformScale=false
    UseOffset=false
    UseSphereOrientation=false
   ) {module=SphereLocation001}
   &CameraOffset(Camera Offset Amount=15)
   &AddVelocity(Velocity=&Multiply_VectorByFloat(Vector=$Output.SphereLocation001.SphereNormal, Float=1))
  ]
  particleUpdateScript [
   &ParticleState()
   &SolveForcesAndVelocity(Clamp Velocity=false, Limit Acceleration=false)
   &Color(
    Color=<rgba 2.5 0.75 0 1>
    Scale Alpha=&FloatFromCurve(FloatCurve=DataInterface() {propertyValues {Curve {keys [<at -0.0046 1.0113 RCIM_Cubic>,<at 1.0011 0.002>]}}})
   )
   &ScaleSpriteSize(Scale Factor=&Vector2DFromCurve(Vector2Curve=DataInterface() {propertyValues {XCurve {keys [<at 0 0.6>,<at 1 2>]}, YCurve {keys [<at 0 0.6>,<at 1 2>]}}}, ScaleCurve=<1 1>))
  ]
  rendererClasses [&NiagaraSpriteRendererProperties]
  renderers [{rendererClass=&NiagaraSpriteRendererProperties}]
 }
 NE_FN_PhysMatDebris {
  emitterUpdateScript [
   &EmitterState()
   &SpawnBurst_Instantaneous(
    Spawn Count=&RandomRangeInt(Minimum=5, Maximum=10)
    Use Spawn Probability=false
    UseLoopCountLimit=false
   )
  ]
  particleSpawnScript [
   /SetVariables(Particles.BiasedRandom=&FloatFromCurve(
     FloatCurve=DataInterface() {propertyValues {Curve {keys [
         <at 0.2 1 RCIM_Cubic>
         {arriveTangent=-3.3394, interpMode=RCIM_Cubic, leaveTangent=-3.3394, tangentMode=RCTM_User, time=0.5, value=0.5}
         <at 0.8 0 RCIM_Cubic>
        ]}}}
     CurveIndex=&RandomRangeFloat(Minimum=0, Maximum=1)
    )) {isSetParameters=true}
   &InitializeParticle(
    Color Mode=::NewEnumerator0
    Lifetime=&RandomRangeFloat(Minimum=0.5, Maximum=1.25)
    Mass=&Lerp_Float(A=0.6, B=1.2, Alpha=$Particles.BiasedRandom)
    Mass Mode=::NewEnumerator1
    Mesh Scale=&Multiply_VectorByFloat(
     Vector=&RandomRangeVector(Minimum=<0.9 0.4 0.5>, Maximum=<1 1 1>)
     Float=&Lerp_Float(A=0.7, B=1, Alpha=$Particles.BiasedRandom)
    )
    Mesh Scale Mode=::NewEnumerator4
    UsePositionOffset=false
    Write Material Random=true
   )
   &FN_PhysicalSurface(
    Dirt Mesh Count=2
    Glass Mesh Count=3
    Ice Mesh Count=2
    Metal Gravity=&RandomRangeFloat(Minimum=-350, Maximum=-1800)
    Metal Mesh Count=2
    Metal Velocity=&RandomRangeFloat(Minimum=80, Maximum=375)
    Metal Velocity Cone Angle=130
    Physical Surface=$User.PhysicalSurface
    Stone Gravity=&RandomRangeFloat(Minimum=-700, Maximum=-1250)
    Stone Mesh Count=2
    Stone Velocity=&RandomRangeFloat(Minimum=80, Maximum=550)
    Stone Velocity Cone Angle=120
    Wood Gravity=&RandomRangeFloat(Minimum=-450, Maximum=-900)
    Wood Mesh Count=2
    Wood Scale=1.5
    Wood Velocity=&RandomRangeFloat(Minimum=80, Maximum=600)
    Wood Velocity Cone Angle=130
   )
   &AddVelocityInCone(
    Cone Angle=$Output.FN_PhysicalSurface.SurfaceResponseConeAngle
    Module.Use Velocity Falloff On Cone Axis=false
    Velocity Strength=$Output.FN_PhysicalSurface.SurfaceResponseVelocity
   )
   &InitialMeshOrientation(Use Rotation Vector=false)
   &AddRotationalVelocity(Rotation Rate=&RandomVector(Vector Scale=&RandomRangeFloat(Minimum=1, Maximum=2)))
   &ApplyInitialForces()
   /SetVariables(Particles.GravityOverrideValue=$Output.FN_PhysicalSurface.GravityOverride) {isSetParameters=true}
  ]
  particleUpdateScript [
   &ParticleState()
   &GravityForce(Gravity=$Particles.GravityOverrideValue)
   &Collision(
    Advanced Aging Rate=1
    Age Colliding Particles=false
    CPU Collision Trace Channel=ECC_WorldStatic
    Collision Enabled=&Boolean_AndOperation(
     Bool1=&SetBoolByFloatComparison(A=$Particles.Initial.Lifetime, B=1)
     Bool2=&SetBoolByFloatComparison(A=$Particles.Age, B=0.12)
    )
    Control Roll On Collision=true
    Friction=1
    Friction During a Bounce=0.25
    Max CPU Trace Length=100
    Module.Control Rest State Via Surface Normal=false
    Module.EnableMaxCollisionCount=false
    Module.Randomize Collision Normal Vector=true
    Randomize Collision Normal=0.25
    ScaleRateOfRotation=false
   ) {enabled=false}
   &SolveForcesAndVelocity(Clamp Velocity=false, Limit Acceleration=false)
   &ScaleMeshSize(Scale Factor=&VectorFromFloat(Value=&FloatFromCurve(FloatCurve=DataInterface() {propertyValues {Curve {keys [<at 0.6 1 RCIM_Cubic>, {arriveTangent=-6.308, interpMode=RCIM_Cubic, leaveTangent=-6.308, tangentMode=RCTM_User, time=1, value=0}]}}})))
   &DynamicMaterialParameters(Module.Param0WriteEnabled=true, Module.Param1WriteEnabled=true, Module.Param2WriteEnabled=false, Module.Param3WriteEnabled=false)
  ]
  rendererClasses [&NiagaraSpriteRendererProperties, &NiagaraMeshRendererProperties]
  renderers [
   {rendererClass=&NiagaraSpriteRendererProperties}
   {rendererClass=&NiagaraMeshRendererProperties, rendererIndex=1}
   {rendererClass=&NiagaraMeshRendererProperties, rendererIndex=2}
   {rendererClass=&NiagaraMeshRendererProperties, rendererIndex=3}
   {rendererClass=&NiagaraMeshRendererProperties, rendererIndex=4}
   {rendererClass=&NiagaraMeshRendererProperties, rendererIndex=5}
   {rendererClass=&NiagaraMeshRendererProperties, rendererIndex=6}
   {rendererClass=&NiagaraMeshRendererProperties, rendererIndex=7}
   {rendererClass=&NiagaraMeshRendererProperties, rendererIndex=8}
  ]
 }
 Smoke {
  emitterUpdateScript [
   &EmitterState()
   &SpawnBurst_Instantaneous(Spawn Count=&UniformRangedInt(Minimum=2, Maximum=2), Use Spawn Probability=false)
  ]
  particleSpawnScript [
   &InitializeParticle(
    Color=&MakeLinearColorFromVectorAndFloat("Vector (RGB)"=<1 1 1>, "Float (Alpha)"=&UniformRangedFloat(Minimum=0.15, Maximum=0.2))
    Lifetime=&UniformRangedFloat(Minimum=0.3, Maximum=0.4)
    Mass Mode=::NewEnumerator1
    Sprite Size=&FN_FrontEndVector2Modifier(
     Front End:Vector2f=&Multiply_Vector2D_ByFloat(
      Float=0.5
      Vector2D=&Vector2DFromFloat(Value=&UniformRangedFloat(Minimum=170, Maximum=250))
     )
     Gameplay:Vector2f=&Vector2DFromFloat(Value=&UniformRangedFloat(Minimum=170, Maximum=250))
    )
    Sprite Size Mode=::NewEnumerator4
    UsePositionOffset=false
    Write Material Random=true
   ) {module=InitializeParticle001}
   &CameraOffset(Camera Offset Amount=&UniformRangedFloat(Minimum=50, Maximum=60))
   /Niagara/Modules/Spawn/Location/SphereLocation(Hemisphere X=true, MaskSpawnGroup=false, Module.Override Randomness=false, Sphere Radius=10, Surface Only=false)
   &AddVelocity(Velocity=&Multiply_Vector(A=&UniformRangedVector(Minimum=<0 100 0>, Maximum=<0 400 0>), B=$System.DirectionVector))
   &AddVelocity(Velocity=&Multiply_Vector(A=$Output.SphereLocation.SphereNormal, B=&VectorFromFloat(Value=5))) {module=AddVelocity001}
  ]
  particleUpdateScript [
   &ParticleState()
   &SolveForcesAndVelocity(Clamp Velocity=false, Limit Acceleration=false)
   /SetVariables(Particles.SubImageIndex=&Lerp_Float(A=0, B=10, Alpha=$Particles.NormalizedAge)) {isSetParameters=true}
   &ScaleSpriteSize(Scale Factor=&Vector2DFromFloat(Value=&FloatFromCurve(
      FloatCurve=DataInterface() {propertyValues {Curve {keys [<at 0 0.05>,<at 1 0.75>]}}}
      Scale Curve=&FN_FrontEndFloatModifier(Gameplay=0.7, Front End=0.4)
     )))
   &DynamicMaterialParameters(
    Index 0 Param 1=&FloatFromCurve(FloatCurve=DataInterface() {propertyValues {Curve {keys [<at 0.5 0.5>,<at 1 -0.1>]}}})
    Index 0 Param 4=1
    Module.Param0WriteEnabled=true
    Module.Param1WriteEnabled=true
    Module.Param2WriteEnabled=true
    Module.Param3WriteEnabled=true
   )
   &ScaleColor(
    Scale Alpha=&FloatFromCurve(FloatCurve=DataInterface() {propertyValues {Curve {keys [<at 0 0>,<at 0.25 1>,<at 1 0>]}}})
    ScaleA=true
    ScaleRGB=true
   )
  ]
  rendererClasses [&NiagaraSpriteRendererProperties]
  renderers [{rendererClass=&NiagaraSpriteRendererProperties}]
 }
 Sparks {
  emitterUpdateScript [
   &EmitterState(Life Cycle Mode=::NewEnumerator1, Loop Behavior=::NewEnumerator1, Loop Delay=0.05, Loop Duration=1, UseLoopDelay=true)
   &SpawnBurst_Instantaneous(
    Spawn Count=&RandomRangeInt(Minimum=10, Maximum=15)
    Use Spawn Probability=false
    UseLoopCountLimit=false
   )
   &SpawnBurst_Instantaneous(Spawn Count=&UniformRangedInt(Minimum=15, Maximum=25), Spawn Group=1, Use Spawn Probability=false) {module=SpawnBurst_Instantaneous001}
  ]
  particleSpawnScript [
   &InitializeParticle(
    Color=&SelectLinearColorFromArray(
     Array Sampling Mode=::NewEnumerator1
     Direct Array Index=$Emitter.SpawnGroup
     Color Selection Array=DataInterface() {propertyValues {ColorData [<rgba 5 2.5 0 1>, <rgba 3 3 3 1>]}}
     Random Seed=?
    )
    Lifetime=&UniformRangedFloat(Minimum=0.1, Maximum=0.35)
    Mass Mode=::NewEnumerator1
    Sprite Size=&FN_FrontEndVector2Modifier(
     Front End:Vector2f=&Multiply_Vector2D_ByFloat(
      Float=0.5
      Vector2D=&Lerp_Vector2(
       A=<3 3>
       B=&UniformRangedVector2D(Minimum=<1 1>, Maximum=<2 2>)
       Alpha=/Game/Effects/Niagara/DynamicInputs/FN_PickFloatFromSpawnGroupCutoff(Over Cutoff Float:NiagaraFloat=1, Under Cutoff Float:NiagaraFloat=0, SpawnGroup:NiagaraInt32=$Emitter.SpawnGroup, Cutoff:NiagaraInt32=1)
      )
     )
     Gameplay:Vector2f=&Lerp_Vector2(
      A=<3 3>
      B=&UniformRangedVector2D(Minimum=<1 1>, Maximum=<2 2>)
      Alpha=&FN_PickFloatFromSpawnGroupCutoff(Over Cutoff Float:NiagaraFloat=1, Under Cutoff Float:NiagaraFloat=0, SpawnGroup:NiagaraInt32=$Emitter.SpawnGroup, Cutoff:NiagaraInt32=1)
     )
    )
    Sprite Size Mode=::NewEnumerator4
    UsePositionOffset=false
    Write Material Random=true
   ) {module=InitializeParticle001}
   &SphereLocation(
    Hemisphere X=true
    MaskSpawnGroup=false
    Module.Override Randomness=false
    Sphere Origin:NiagaraPosition=&SimulationPosition()
    Sphere Radius=&SelectFloatFromArray(Direct Array Index=$Emitter.SpawnGroup, Float Selection Array=DataInterface() {propertyValues {FloatData [8,20]}})
    Surface Only=false
   )
   &AddVelocity(Velocity=&Multiply_Vector(
     A=$Output.SphereLocation.SphereNormal
     B=&VectorFromFloat(Value=&Lerp_Float(
       A=&UniformRangedFloat(Minimum=12, Maximum=70)
       B=&UniformRangedFloat(Minimum=5, Maximum=20)
       Alpha=&FN_PickFloatFromSpawnGroupCutoff(Over Cutoff Float:NiagaraFloat=1, Under Cutoff Float:NiagaraFloat=0, SpawnGroup:NiagaraInt32=$Emitter.SpawnGroup, Cutoff:NiagaraInt32=1)
      ))
    )) {module=AddVelocity001}
   &AddVelocity(Velocity=&FN_FrontEndVector3Modifier(
     Gameplay=&Multiply_Vector(
      A=&Lerp_Vector(
       A=&UniformRangedVector(
        Minimum=&Multiply_Vector(A=<0 1000 -200>, B=&VectorFromFloat(Value=0.1))
        Maximum=<0 1000 -200>
       )
       B=&UniformRangedVector(Minimum=<0 100 -200>, Maximum=<0 1000 -200>)
       Alpha=&FN_PickFloatFromSpawnGroupCutoff(Over Cutoff Float:NiagaraFloat=1, Under Cutoff Float:NiagaraFloat=0, SpawnGroup:NiagaraInt32=$Emitter.SpawnGroup, Cutoff:NiagaraInt32=1)
      )
      B=$System.DirectionVector
     )
     Front End=&Multiply_Vector(
      A=&Lerp_Vector(
       A=&UniformRangedVector(
        Minimum=&Multiply_Vector(A=<200 1000 -200>, B=&VectorFromFloat(Value=0.1))
        Maximum=<200 1000 -200>
       )
       B=&UniformRangedVector(Minimum=<200 100 -200>, Maximum=<300 1000 -200>)
       Alpha=&FN_PickFloatFromSpawnGroupCutoff(Over Cutoff Float:NiagaraFloat=1, Under Cutoff Float:NiagaraFloat=0, SpawnGroup:NiagaraInt32=$Emitter.SpawnGroup, Cutoff:NiagaraInt32=1)
      )
      B=$System.DirectionVector
     )
    ))
   &CameraOffset(Camera Offset Amount=25)
  ]
  particleUpdateScript [
   &ParticleState()
   &AccelerationForce(Acceleration=&SelectVectorFromArray(Direct Array Index=$Emitter.SpawnGroup, Vector Selection Array=DataInterface() {propertyValues {FloatData [<0 0 -1100>, <0 0 -500>]}}))
   &SolveForcesAndVelocity(Clamp Velocity=false, Limit Acceleration=false)
   &ScaleSpriteSizeBySpeed(
    Max Scale Factor=&MakeVector2D(
     X=&FN_PickFloatFromSpawnGroupCutoff(Over Cutoff Float:NiagaraFloat=1, Under Cutoff Float:NiagaraFloat=0.5, SpawnGroup:NiagaraInt32=$Emitter.SpawnGroup, Cutoff:NiagaraInt32=1)
     Y=&FN_PickFloatFromSpawnGroupCutoff(Over Cutoff Float:NiagaraFloat=3, Under Cutoff Float:NiagaraFloat=10, SpawnGroup:NiagaraInt32=$Emitter.SpawnGroup, Cutoff:NiagaraInt32=1)
    )
    Min Scale Factor=&MakeVector2D(
     X=&FN_PickFloatFromSpawnGroupCutoff(Over Cutoff Float:NiagaraFloat=1, Under Cutoff Float:NiagaraFloat=0.5, SpawnGroup:NiagaraInt32=$Emitter.SpawnGroup, Cutoff:NiagaraInt32=1)
     Y=&FN_PickFloatFromSpawnGroupCutoff(Over Cutoff Float:NiagaraFloat=1, Under Cutoff Float:NiagaraFloat=0.5, SpawnGroup:NiagaraInt32=$Emitter.SpawnGroup, Cutoff:NiagaraInt32=1)
    )
   )
   &ScaleSpriteSize(Scale Factor=&Vector2DFromFloat(Value=&FloatFromCurve(FloatCurve=DataInterface() {propertyValues {Curve {keys [{arriveTangent=-6.725, interpMode=RCIM_Cubic, leaveTangent=-6.725, tangentMode=RCTM_User, time=0.65, value=1}, <at 1 0 RCIM_Cubic>]}}})))
   &ScaleColor(
    Scale RGB=&Lerp_Vector(
     A=&VectorFromCurve(VectorCurve=DataInterface() {propertyValues {
        XCurve {keys [<at 0.5 12 RCIM_Cubic>,<at 1 1 RCIM_Cubic>]}
        YCurve {keys [<at 0.5 6 RCIM_Cubic>,<at 1 1 RCIM_Cubic>]}
        ZCurve {keys [<at 0.5 2 RCIM_Cubic>,<at 1 1 RCIM_Cubic>]}
       }})
     B=&MakeVectorFromLinearColorRGB(Color=&ColorFromCurve(ColorCurve=DataInterface() {propertyValues {
         AlphaCurve {keys [<at 0 1 RCIM_Cubic>,<at 1 1 RCIM_Cubic>]}
         BlueCurve {keys [<at 0.2 0.6127 RCIM_Cubic>,<at 0.7 0.045 RCIM_Cubic>]}
         GreenCurve {keys [<at 0.2 3.4085 RCIM_Cubic>,<at 0.7 0.045 RCIM_Cubic>]}
         RedCurve {keys [<at 0.2 20 RCIM_Cubic>,<at 0.7 0.5 RCIM_Cubic>]}
        }}))
     Alpha=&FN_PickFloatFromSpawnGroupCutoff(Over Cutoff Float:NiagaraFloat=1, Under Cutoff Float:NiagaraFloat=0, SpawnGroup:NiagaraInt32=$Emitter.SpawnGroup, Cutoff:NiagaraInt32=1)
    )
    ScaleA=true
    ScaleRGB=true
   )
  ]
  rendererClasses [&NiagaraSpriteRendererProperties]
  renderers [{rendererClass=&NiagaraSpriteRendererProperties}]
 }
 Spike_big {
  emitterUpdateScript [
   &EmitterState()
   &SpawnBurst_Instantaneous(Spawn Count=1, Use Spawn Probability=false)
  ]
  particleSpawnScript [
   &InitializeParticle(
    Color=<rgba 5 1.5 0 1>
    Lifetime=0.1
    Mass Mode=::NewEnumerator1
    Sprite Size=&FN_FrontEndVector2Modifier(
     Front End:Vector2f=&Multiply_Vector2(A=<70 70>, B=&Vector2DFromFloat(Value=0.4))
     Gameplay:Vector2f=&Multiply_Vector2(A=<70 70>, B=&Vector2DFromFloat(Value=0.7))
    )
    Sprite Size Mode=::NewEnumerator4
    UsePositionOffset=false
    Write Material Random=true
   ) {module=InitializeParticle001}
   &AddVelocity(Velocity=&FN_FrontEndVector3Modifier(
     Gameplay=&Multiply_Vector(A=<0 1000 -200>, B=$System.DirectionVector)
     Front End=&Multiply_Vector(A=<300 1000 -200>, B=$System.DirectionVector)
    )) {module=AddVelocity001}
   /SetVariables(Particles.SubImageIndex=3) {isSetParameters=true}
  ]
  particleUpdateScript [
   &ParticleState()
   &SolveForcesAndVelocity(Clamp Velocity=false, Limit Acceleration=false)
   &ScaleColor(
    Scale Alpha=&Lerp_Float(A=1, B=0, Alpha=$Particles.NormalizedAge)
    Scale RGB=&VectorFromFloat(Value=2)
    ScaleA=true
    ScaleRGB=true
   )
   &CameraOffset(Camera Offset Amount=50)
   &ScaleSpriteSize(Scale Factor=&Vector2DFromCurve(Vector2Curve=DataInterface() {propertyValues {XCurve {keys [<at 0 1>,<at 1 0>]}, YCurve {keys [<at 0 1>,<at 1 2>]}}}, ScaleCurve=<1 1>))
   &DynamicMaterialParameters(
    Index 0 Param 1=50
    Module.Param0WriteEnabled=true
    Module.Param1WriteEnabled=true
    Module.Param2WriteEnabled=true
    Module.Param3WriteEnabled=true
   ) {module=DynamicMaterialParameters001}
  ]
  rendererClasses [&NiagaraSpriteRendererProperties]
  renderers [{rendererClass=&NiagaraSpriteRendererProperties}]
 }
}
summary {
 systemName=NS_Pickaxe_ShrimpStroll_DirectionalImpact_Final
 emitters [
  {emitterName=Smoke, rendererClasses [&NiagaraSpriteRendererProperties]}
  {emitterName=Spike_big, rendererClasses [&NiagaraSpriteRendererProperties]}
  {emitterName=MIN_Flash001, rendererClasses [&NiagaraSpriteRendererProperties]}
  {
   emitterName=NE_FN_PhysMatDebris
   rendererClasses [&NiagaraSpriteRendererProperties, &NiagaraMeshRendererProperties]
  }
  {emitterName=Sparks, rendererClasses [&NiagaraSpriteRendererProperties]}
  {
   emitterName=Flash_Flare001
   rendererClasses [
    &NiagaraSpriteRendererProperties
    &NiagaraLightRendererProperties
    &NiagaraMeshRendererProperties
   ]
  }
  {emitterName=Leaves, rendererClasses [&NiagaraSpriteRendererProperties]}
 ]
 userVariables [
  User.bIsFrontEnd {type=NiagaraBool}
  User.bIsFrontEndPreview {type=NiagaraBool}
  User.HitVegetation? {type=NiagaraBool}
  User.LandBurst {type=NiagaraFloat, defaultValue=0, description=""}
  User.PhysicalSurface {type=NiagaraInt32}
  User.SwingSectionIndex {type=NiagaraInt32, defaultValue=3}
 ]
}
